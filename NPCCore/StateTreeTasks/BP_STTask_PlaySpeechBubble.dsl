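; State Tree task that makes an NPC pawn play a tagged speech bubble after a delay.
; Tick counts down the delay, then casts either the AI controller's pawn or, without a controller, the supplied pawn actor to FortPawn.
; The pawn-actor branch adds a speech bubble component if missing, calls PlaySpeechBubble with stop-previous and matching-VO options, then finishes the task.
; As shown, the controller branch only stores the pawn.
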(event EventEnterState
  (params (Transition :type State%20Tree%20Transition%20Result%20Structure%20%28by%20ref%29)))

(event EventTick
  (params (DeltaTime :type Float%20%28single-precision%29))
  (bind _pawnactor (Variables|Default|GetPawnActor))
  (bind _aipawn (Variables|Default|GetAIPawn))
  (bind _aicontroller (Variables|Default|GetAIController))
  ; Delay
  (bind _output_get (Variables|Default|SetRemainingDelayTime (- (Variables|Default|GetRemainingDelayTime) DeltaTime)))
  (if (<= _output_get 0)
    ; Play Speech Bubble logic
    (Utilities|IsValid _aicontroller
      (:Is%20Valid
        (bind _asfort_athena_aibot_controller (Utilities|Casting|CastToFortAthenaAIBotController _aicontroller)
          (:then
            (bind _asfort_pawn (Utilities|Casting|CastToFortPawn (Pawn|GetControlledPawn _asfort_athena_aibot_controller))
              (:then
                (Variables|Default|SetAIPawn _asfort_pawn))
              (:CastFailed
                (StateTree|FinishTask false))))
          (:CastFailed
            (StateTree|FinishTask false))))
      (:Is%20Not%20Valid
        (Utilities|IsValid _pawnactor
          (:Is%20Valid
            (bind _asfort_pawn_2 (Utilities|Casting|CastToFortPawn _pawnactor)
              (:then
                (Variables|Default|SetAIPawn _asfort_pawn_2)
                (Utilities|IsValid (Actor|GetComponentbyClass _aipawn "/Script/FortniteGame.FortPawnComponent_SpeechBubble")
                  (:Is%20Valid
                    _)
                  (:Is%20Not%20Valid
                    (bind _returnvalue (Game|AddComponentbyClass (Variables|Default|GetAIPawn) (Variables|Default|GetSpeechComponent)))
                    (Utilities|IsValid _returnvalue
                      (:Is%20Valid
                        _)
                      (:Is%20Not%20Valid
                        (StateTree|FinishTask false)))))
                (Class|BPFLAICommonFunctions|PlaySpeechBubble _aipawn (Variables|Default|GetSpeechBubbleTag) (Variables|Default|GetStopPreviousSpeechBubble) (Variables|Default|GetTryToPlayVOLineWithMatchingTag))
                (StateTree|FinishTask true))
              (:CastFailed
                _)))
          (:Is%20Not%20Valid
            _))
        (StateTree|FinishTask false)))))
